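# Niagara system for the deploy effect of the MincePounce glider: a delayed burst of two blades, their flares and a thruster ignition.
# Socket-tracked BladesDeploy meshes and BladeFlares sprites use particle partitioning to split left and right with mirrored offsets.
# MainThruster_Flare and curl-noise Sparks share the thruster socket, offset and burst delay from system variables.
# Variant selectors choose colors, and front-end modifiers adjust some thruster offsets.
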NiagaraSystem /BRCosmetics/Gadgets/Assets/VinderTech_GliderChute/Glider_MincePounce/FX/NS_Glider_MincePounce_Deploy
dependencies {
 usedDataInterfaces [
  &NiagaraDataInterfaceCurve
  &NiagaraDataInterfaceColorCurve
  &NiagaraDataInterfaceSocketReader
  &NiagaraDataInterfaceArrayFloat
 ]
 usedDynamicInputs [
  &SetBoolByFloatComparison
  &RemapRange
  &SelectFloatFromArray
  &MakeIntFromFloat
  &NormalizeVector
  &Add_Vector
  &TransformVector
  &Waveform
 ]
 usedModules [
  &self:SystemSpawnScript.SystemScriptSource.SystemScriptGraph.NiagaraNodeAssignment_1.SetVariables_#1
  &self:SystemSpawnScript.SystemScriptSource.SystemScriptGraph.NiagaraNodeAssignment_3.SetVariables_#2
  &self:SystemSpawnScript.SystemScriptSource.SystemScriptGraph.NiagaraNodeAssignment_2.SetVariables_#3
  &SystemState
  &Timeline
  &FN_Integrator
  &InterpolateOverTime
  &self:SystemSpawnScript.SystemScriptSource.SystemScriptGraph.NiagaraNodeAssignment_0.SetVariables_#4
  &FN_VariantManager
 ]
}
emitters {
 BladeFlares {
  emitterUpdateScript [
   &EmitterState(Life Cycle Mode=::NewEnumerator1, Loop Behavior=::NewEnumerator1, Loop Delay=0.9, Loop Duration=1, UseLoopDelay=true)
   &SpawnBurst_Instantaneous(Spawn Count=2, Use Spawn Probability=false, UseLoopCountLimit=true)
  ]
  particleSpawnScript [
   &PartitionParticles()
   &InitializeParticle(
    Color=&VariantColorSelector(Variant Amount=0, Original Color=<rgba 0.3889 0.2174 3 1>, Variant 1 Color=<rgba 0.2174 1.6087 3 1>, Variant 2 Color=?)
    Lifetime=0.3
    Sprite Size=&Multiply_Vector2D_ByFloat(Float=150, Vector2D=<1 1>)
    Sprite Size Mode=::NewEnumerator4
    UsePositionOffset=false
    Write Material Random=true
   )
   &AddVelocity(Velocity=<0 100 0>) {enabled=false}
   &SpriteFacingAndAlignment(Alignment Coordinate Space=::Local, WriteAlignment=true, WriteFacing=false) {enabled=false}
   &CameraOffset(Camera Offset Amount=15)
   /SetVariables(Particles.CameraOffsetFlare=150) {isSetParameters=true}
  ]
  particleUpdateScript [
   &SocketLocation(
    Interpolate Socket Sampling=true
    SamplePosOffset=true
    Sampled Position Offset=&AddVectorToPosition(
     Position=<0 0 0>
     Vector=&MakeVector(
      X=&MergeHelper_FloatSwitch(0=40, 1=-40, 2=?, 3=?, Index=$Particles.PartitionParticles.Partition, Num=::NewEnumerator0)
      Y=0
      Z=0
     )
    )
    Socket Index=$Particles.PartitionParticles.Partition
    Socket Reader=$System.Socket Reader
    Socket Sampling Mode=::NewEnumerator3
    VelocitySpeedLimit=false
   ) {enabled=false}
   &SocketLocation(
    Interpolate Socket Sampling=true
    SamplePosOffset=true
    Sampled Position Offset=&AddVectorToPosition(
     Position=<0 0 0>
     Vector=&MakeVector(
      X=&FloatFromCurve(
       FloatCurve=DataInterface() {propertyValues {Curve {keys [
           {interpMode=RCIM_Cubic, tangentMode=RCTM_User, time=0, value=0.33}
           {arriveTangent=0.8622, interpMode=RCIM_Cubic, leaveTangent=0.8622, tangentMode=RCTM_User, time=1, value=1}
          ]}}}
       Scale Curve=&MergeHelper_FloatSwitch(0=125, 1=-125, 2=?, 3=?, Index=$Particles.PartitionParticles.Partition, Num=::NewEnumerator0)
      )
      Y=0
      Z=&FloatFromCurve(
       FloatCurve=DataInterface() {propertyValues {Curve {keys [
           {arriveTangent=0.2458, interpMode=RCIM_Cubic, leaveTangent=0.2458, tangentMode=RCTM_User, time=0, value=0}
           {arriveTangent=0.5331, interpMode=RCIM_Cubic, leaveTangent=0.5331, tangentMode=RCTM_User, time=1, value=0.9}
          ]}}}
       Scale Curve=&MergeHelper_FloatSwitch(0=25, 1=-25, 2=?, 3=?, Index=$Particles.PartitionParticles.Partition, Num=::NewEnumerator0)
      )
     )
    )
    Socket Index=$Particles.PartitionParticles.Partition
    Socket Reader=$System.Socket Reader
    Socket Sampling Mode=::NewEnumerator3
    VelocitySpeedLimit=false
   ) {module=SocketLocation001}
   &ParticleState()
   &ScaleSpriteSize(Uniform Curve Sprite Scale=DataInterface() {propertyValues {Curve {keys [{interpMode=RCIM_Cubic, tangentMode=RCTM_User, time=0, value=1}, {interpMode=RCIM_Cubic, tangentMode=RCTM_User, time=1, value=0.2}]}}})
   &SolveForcesAndVelocity(Clamp Velocity=false, Limit Acceleration=false)
   &ScaleColor(
    Scale Alpha=&FloatFromCurve(FloatCurve=DataInterface() {propertyValues {Curve {keys [
         <at 0 0 RCIM_Cubic>
         {arriveTangent=-0.5651, interpMode=RCIM_Cubic, leaveTangent=-0.5651, tangentMode=RCTM_User, time=0.1, value=1}
         {arriveTangent=-0.5113, interpMode=RCIM_Cubic, leaveTangent=-0.5113, tangentMode=RCTM_User, time=0.8, value=0}
        ]}}})
    Scale RGB=&MakeVectorFromLinearColorRGB(Color=&ColorFromCurve(ColorCurve=DataInterface() {propertyValues {
        AlphaCurve {keys [<at 0 1 RCIM_Cubic>]}
        BlueCurve {keys [<at 0 5 RCIM_Cubic>,<at 0.6102 5>]}
        GreenCurve {keys [<at 0 2.2292 RCIM_Cubic>,<at 0.6102 0.3117>]}
        RedCurve {keys [<at 0 0.25 RCIM_Cubic>,<at 0.6102 0.8943>]}
       }}))
    ScaleA=true
    ScaleRGB=false
   )
   /SetVariables(
    Particles.ColorGlow:LinearColor=&MultiplyLinearColor(Linear Color=$Particles.Color, Scale Factor=&MakeVector4_FromLinearColor(LinearColor=<rgba 3 3 3 1>))
    Particles.SpriteSizeGlow:Vector2f=&Multiply_Vector2D_ByFloat(Float=2, Vector2D=$Particles.SpriteSize)
   ) {isSetParameters=true}
  ]
  rendererClasses [&NiagaraSpriteRendererProperties]
  renderers [
   {rendererClass=&NiagaraSpriteRendererProperties}
   {rendererClass=&NiagaraSpriteRendererProperties, rendererIndex=1}
  ]
 }
 BladesDeploy {
  emitterUpdateScript [
   &EmitterState(Life Cycle Mode=::NewEnumerator1, Loop Behavior=::NewEnumerator1, Loop Delay=0.86, Loop Duration=1, UseLoopDelay=true)
   &SpawnBurst_Instantaneous(Spawn Count=2, Use Spawn Probability=false, UseLoopCountLimit=false)
  ]
  particleSpawnScript [
   &PartitionParticles()
   &InitializeParticle(
    Color=&VariantColorSelector(Variant Amount=0, Original Color=<rgba 1.6716 0.3432 3 1>, Variant 1 Color=<rgba 0.3432 1.6716 3 1>, Variant 2 Color=?)
    Lifetime=0.33
    Mesh Scale=&Multiply_VectorByFloat(
     Vector=&MergeHelper_VectorSwitch(0=<1.1 1 2>, 1=<1.1 1.1 2>, 2=?, 3=?, Index=$Particles.PartitionParticles.Partition, Num=::NewEnumerator0)
     Float=1
    )
    Mesh Scale Mode=::NewEnumerator4
    UsePositionOffset=false
    Write Material Random=true
   )
   &InitialMeshOrientation(
    Mesh Axis To Orient=&MergeHelper_VectorSwitch(0=<0 0 1>, 1=<0 0 -1>, 2=?, 3=?, Index=$Particles.PartitionParticles.Partition, Num=::NewEnumerator0)
    Mesh Orientation Mode=::NewEnumerator2
    Rotation=&MergeHelper_VectorSwitch(0=<-0.25 0 0>, 1=<-0.25 0 0>, 2=?, 3=?, Index=$Particles.PartitionParticles.Partition, Num=::NewEnumerator0)
    Use Rotation Vector=true
   )
   &InitialMeshOrientation(Mesh Orientation Mode=::NewEnumerator2, Rotation=<-0.02 0.02 -0.01>, Use Rotation Vector=true) {module=InitialMeshOrientation001}
   &CameraOffset(Camera Offset Amount=15)
  ]
  particleUpdateScript [
   &ParticleState()
   &ScaleMeshSize(Scale Factor=&VectorFromFloat(Value=&FloatFromCurve(FloatCurve=DataInterface() {propertyValues {Curve {keys [
          {arriveTangent=1.5185, interpMode=RCIM_Cubic, leaveTangent=1.5185, tangentMode=RCTM_User, time=0, value=0.3}
          {interpMode=RCIM_Cubic, tangentMode=RCTM_User, time=0.6, value=1}
         ]}}}))) {module=ScaleMeshSize001}
   &SocketLocation(
    Interpolate Socket Sampling=true
    SamplePosOffset=true
    Sampled Position Offset=&MergeHelper_VectorSwitch(0=<5 0 5>, 1=<-5 0 5>, 2=?, 3=?, Index=$Particles.PartitionParticles.Partition, Num=::NewEnumerator0)
    Socket Index=$Particles.PartitionParticles.Partition
    Socket Reader=$System.Socket Reader
    Socket Sampling Mode=::NewEnumerator3
    VelocitySpeedLimit=false
   )
   &ScaleMeshSize(Scale Factor=&MakeVector(
     X=1
     Y=1
     Z=&FloatFromCurve(FloatCurve=DataInterface() {propertyValues {Curve {keys [<at 0 2>,<at 1 0>]}}})
    ))
   &DynamicMaterialParameters(
    Index 0 Param 3=&OneMinusFloat(Float=&FloatFromCurve(FloatCurve=DataInterface() {propertyValues {Curve {keys [
          {arriveTangent=-1.7386, interpMode=RCIM_Cubic, leaveTangent=-1.7386, tangentMode=RCTM_User, time=0, value=1}
          {arriveTangent=-0.041, tangentMode=RCTM_User, time=1, value=0}
         ]}}}))
    Module.Index0Float0Write=false
    Module.Index0Float1Write=false
    Module.Index0Float2Write=true
    Module.Index0Float3Write=false
   )
   &DynamicMaterialParameters(
    Index 0 Param 1=3
    Index 0 Param 2=0.4
    Module.Index0Float0Write=true
    Module.Index0Float1Write=true
    Module.Index0Float2Write=true
    Module.Index0Float3Write=true
   ) {enabled=false, module=DynamicMaterialParameters001}
   &ScaleColor(
    Scale Alpha=&FloatFromCurve(FloatCurve=DataInterface() {propertyValues {Curve {keys [
         {arriveTangent=8.0416, interpMode=RCIM_Cubic, leaveTangent=8.0416, tangentMode=RCTM_User, time=0, value=0}
         {arriveTangent=-2.8913, interpMode=RCIM_Cubic, leaveTangent=-2.8913, tangentMode=RCTM_User, time=0.1, value=1}
         {interpMode=RCIM_Cubic, tangentMode=RCTM_User, time=1, value=0}
        ]}}})
    ScaleA=true
    ScaleRGB=true
   )
   /SetVariables(Particles.MeshOrientationBlade:Quat4f=&MultiplyQuaternion(
     Quaternion A=$Particles.MeshOrientation
     Quaternion B=&MakeQuaternion(
      Quaternion From=0
      Axis=<1 0 0>
      Angle=&FloatFromCurve(
       FloatCurve=DataInterface() {propertyValues {Curve {keys [
           {interpMode=RCIM_Cubic, tangentMode=RCTM_User, time=0.1, value=-0.2}
           {arriveTangent=2.0716, interpMode=RCIM_Cubic, leaveTangent=2.0716, tangentMode=RCTM_User, time=1, value=1}
          ]}}}
       Scale Curve=&MergeHelper_FloatSwitch(0=0, 1=0, 2=?, 3=?, Index=$Particles.PartitionParticles.Partition, Num=::NewEnumerator0)
      )
      From=?
      To=?
      Angle Type=::NewEnumerator1
      Normalize Vectors=false
     )
    )) {isSetParameters=true}
  ]
  rendererClasses [&NiagaraMeshRendererProperties]
  renderers [{rendererClass=&NiagaraMeshRendererProperties, rendererIndex=0}]
 }
 MainThruster_Flare {
  emitterUpdateScript [
   &EmitterState(Life Cycle Mode=::NewEnumerator1, Loop Behavior=::NewEnumerator1, Loop Delay=$System.ThrusterBurstDelay, Loop Duration=1, UseLoopDelay=true)
   &SpawnBurst_Instantaneous(Spawn Count=1, Use Spawn Probability=false)
  ]
  particleSpawnScript [
   &InitializeParticle(
    Color=&VariantColorSelector(Variant Amount=0, Original Color=<rgba 0.0815 0.2958 1 1>, Variant 1 Color=<rgba 0.3947 0.192 0.8 1>, Variant 2 Color=?)
    Lifetime=0.93
    Position Offset=<0 73 0>
    Sprite Size=&Multiply_Vector2D_ByFloat(Float=200, Vector2D=<1.5 1>)
    Sprite Size Mode=::NewEnumerator4
    UsePositionOffset=false
    Write Material Random=true
   )
   &DynamicMaterialParameters(
    Index 0 Param 1=1
    Index 0 Param 2=1
    Index 0 Param 3=1
    Index 0 Param 4=1
    Module.Param0WriteEnabled=true
    Module.Param1WriteEnabled=true
    Module.Param2WriteEnabled=true
    Module.Param3WriteEnabled=true
   )
   &CameraOffset(Camera Offset Amount=&FN_FrontEndFloatModifier(Gameplay=50, Front End=10))
  ]
  particleUpdateScript [
   &SocketLocation(
    SamplePosOffset=true
    Sampled Position Offset=$System.ThrusterOffset
    Socket Index=2
    Socket Reader=$System.Socket Reader
    Socket Sampling Mode=::NewEnumerator3
    VelocitySpeedLimit=false
   )
   &ParticleState()
   &ScaleSpriteSize(Uniform Curve Sprite Scale=DataInterface() {propertyValues {Curve {keys [
        {arriveTangent=9.0725, interpMode=RCIM_Cubic, leaveTangent=9.0725, tangentMode=RCTM_User, time=0, value=0.6}
        {arriveTangent=-0.2121, interpMode=RCIM_Cubic, leaveTangent=-0.2121, tangentMode=RCTM_User, time=0.1408, value=0.9847}
        {arriveTangent=-0.2714, interpMode=RCIM_Cubic, leaveTangent=-0.2714, tangentMode=RCTM_User, time=1, value=0.8}
       ]}}})
   &ScaleColor(
    Scale Alpha=&FloatFromCurve(
     FloatCurve=DataInterface() {propertyValues {Curve {keys [
         {arriveTangent=5.7004, interpMode=RCIM_Cubic, leaveTangent=5.7004, tangentMode=RCTM_User, time=0, value=0}
         {arriveTangent=-2.6665, interpMode=RCIM_Cubic, leaveTangent=-2.6665, tangentMode=RCTM_User, time=0.13, value=1}
         {arriveTangent=-0.7707, interpMode=RCIM_Cubic, leaveTangent=-0.7707, tangentMode=RCTM_User, time=0.26, value=0.3}
         {arriveTangent=-0.0405, interpMode=RCIM_Cubic, leaveTangent=-0.0405, tangentMode=RCTM_User, time=1, value=0}
        ]}}}
     Scale Curve=0.45
    )
    Scale RGB=&MakeVectorFromLinearColorRGB(Color=&ColorFromCurve(ColorCurve=DataInterface() {propertyValues {
        AlphaCurve {keys [<at 0 1 RCIM_Cubic>]}
        BlueCurve {keys [<at 0 4 RCIM_Cubic>,<at 0.3598 1 RCIM_Cubic>]}
        GreenCurve {keys [<at 0 4 RCIM_Cubic>,<at 0.3598 1 RCIM_Cubic>]}
        RedCurve {keys [<at 0 4 RCIM_Cubic>,<at 0.3598 1 RCIM_Cubic>]}
       }}))
    ScaleA=true
    ScaleRGB=true
   )
   /SetVariables(
    Particles.ColorGlow:LinearColor=&MultiplyLinearColor(
     Linear Color=$Particles.Color
     Scale Factor=&Multiply_Vector4(
      A=&MakeVector4_FromLinearColor(LinearColor=&VariantColorSelector(Variant Amount=0, Original Color=<rgba 0.3333 0.2 1 2>, Variant 1 Color=<rgba 0.4 0.7 1 2>, Variant 2 Color=?))
      B=&MakeVector4FromVectorAndFloat(
       "Vector (RGB)"=<1 1 1>
       "Float (Alpha)"=&FloatFromCurve(FloatCurve=DataInterface() {propertyValues {Curve {keys [{interpMode=RCIM_Cubic, tangentMode=RCTM_User, time=0.4, value=1}, {interpMode=RCIM_Cubic, tangentMode=RCTM_User, time=0.7, value=0}]}}})
      )
     )
    )
    Particles.SpriteSizeGlow:Vector2f=&Multiply_Vector2D_ByFloat(Float=0.5, Vector2D=$Particles.SpriteSize)
   ) {isSetParameters=true}
  ]
  rendererClasses [&NiagaraSpriteRendererProperties]
  renderers [
   {rendererClass=&NiagaraSpriteRendererProperties}
   {rendererClass=&NiagaraSpriteRendererProperties, rendererIndex=1}
  ]
 }
 Sparks {
  emitterUpdateScript [
   &EmitterState(
    Life Cycle Mode=::NewEnumerator1
    Loop Behavior=::NewEnumerator1
    Loop Delay=&Add_Float(A=$System.ThrusterBurstDelay, B=0)
    Loop Duration=1
    UseLoopDelay=true
   )
   &SpawnBurst_Instantaneous(
    Spawn Count=&RandomRangeInt(Minimum=18, Maximum=30)
    Use Spawn Probability=false
    UseLoopCountLimit=false
   )
  ]
  particleSpawnScript [
   &InitializeParticle(
    Color=&VariantColorSelector(Variant Amount=0, Original Color=<rgba 0.135 0.135 1.5 1>, Variant 1 Color=<rgba 0.0591 0.7795 1.5 1>, Variant 2 Color=?)
    Lifetime=&RandomRangeFloat(Minimum=0.3, Maximum=1)
    Sprite Size=&Multiply_Vector2D_ByFloat(Float=&RandomRangeFloat(Minimum=0.3, Maximum=0.6), Vector2D=<3 5>)
    Sprite Size Mode=::NewEnumerator4
    UsePositionOffset=false
    Write Material Random=true
   )
   &SocketLocation(SamplePosOffset=false, Socket Index=2, Socket Reader=$System.Socket Reader, Socket Sampling Mode=::NewEnumerator3, VelocitySpeedLimit=false)
   &ShapeLocation(
    MaskSpawnGroup=false
    Non Uniform Scale=<1 0.5 0.3>
    Override Seed=false
    Shape Origin=&AddVectorToPosition(
     Position=$Output.SocketLocation.SampledPosition
     Vector=&RotateVectorByQuaternion(
      Vector To Rotate=&AddVectorToPosition(
       Position=Linked() {linkedVariable {name=System.ThrusterOffset, type=Vector3f}}
       Vector=&FN_FrontEndVector3Modifier(Gameplay=<0 -10 0>, Front End=<0 -10 -2>)
      )
      Quaternion=$Output.SocketLocation.SampledRotation
     )
    )
    Sphere Radius=23
    Sphere Surface Distribution=0.2619
    UseOwnerScale=true
   )
   &AddVelocity(
    Module.Use Falloff Distance=false
    Origin Offset=&RotateVectorByQuaternion(
     Vector To Rotate=&RandomRangeVector(Minimum=<0 -100 0>, Maximum=<0 -150 0>)
     Quaternion=$Output.SocketLocation.SampledRotation
    )
    Origin Offset Coordinate Space=::Simulation
    Override Seed=false
    Velocity Mode=::NewEnumerator1
    Velocity Origin=$Output.SocketLocation.SampledPosition
    Velocity Speed=&RandomRangeFloat(Minimum=50, Maximum=400)
   )
   &CameraOffset(Camera Offset Amount=3)
   &AddVelocity(
    Module.Use Falloff Distance=false
    Override Seed=false
    Velocity Mode=::NewEnumerator1
    Velocity Origin=$Output.SocketLocation.SampledPosition
    Velocity Speed=100
   ) {enabled=false, module=AddVelocity001}
   &AddVelocity(
    Velocity=&RotateVectorByQuaternion(Vector To Rotate=<0 1 0>, Quaternion=$Output.SocketLocation.SampledRotation)
    Velocity Speed Scale=&RandomRangeFloat(Minimum=100, Maximum=300)
   ) {enabled=false, module=AddVelocity002}
  ]
  particleUpdateScript [
   &ParticleState()
   &ScaleSpriteSize(Uniform Curve Sprite Scale=DataInterface() {propertyValues {Curve {keys [{tangentMode=RCTM_User, time=0.3, value=1}, {tangentMode=RCTM_User, time=1, value=0}]}}})
   &Drag(Drag=&RandomRangeFloat(Minimum=1.5, Maximum=4), Use Linear Drag=true, Use Rotational Drag=true)
   &CurlNoiseForce(
    Bias Noise=false
    Noise Frequency=100
    Noise Strength=&RandomRangeFloat(Minimum=400, Maximum=600)
    Pan Noise=true
    Pan Noise Field=<0.1 0 0>
   )
   &InheritSourceMovement(ApplyForce=false, ApplyPosition=true, ApplyPositionFalloff=false)
   &SolveForcesAndVelocity(Clamp Velocity=false, Limit Acceleration=false)
   &ScaleSpriteSizeBySpeed(Max Scale Factor=<1 2>, Min Scale Factor=<1 0.5>, Velocity Threshold=200)
   &ScaleColor(
    Scale RGB=&MakeVectorFromLinearColorRGB(Color=&ColorFromCurve(ColorCurve=DataInterface() {propertyValues {
        AlphaCurve {keys [<at 0 1 RCIM_Cubic>]}
        BlueCurve {keys [<at 0.1179 10 RCIM_Cubic>,<at 0.739 1 RCIM_Cubic>]}
        GreenCurve {keys [<at 0.1179 10 RCIM_Cubic>,<at 0.739 0.464 RCIM_Cubic>]}
        RedCurve {keys [<at 0.1179 10 RCIM_Cubic>,<at 0.739 0.3182 RCIM_Cubic>]}
       }}))
    ScaleA=true
    ScaleRGB=true
   )
  ]
  rendererClasses [&NiagaraSpriteRendererProperties]
  renderers [{rendererClass=&NiagaraSpriteRendererProperties}]
 }
}
summary {
 systemName=NS_Glider_MincePounce_Deploy
 emitters [
  {emitterName=BladesDeploy, rendererClasses [&NiagaraMeshRendererProperties]}
  {emitterName=MainThruster_Flare, rendererClasses [&NiagaraSpriteRendererProperties]}
  {emitterName=BladeFlares, rendererClasses [&NiagaraSpriteRendererProperties]}
  {emitterName=Sparks, rendererClasses [&NiagaraSpriteRendererProperties]}
 ]
 userVariables [
  User.bIsFrontEnd {type=NiagaraBool}
  User.bIsFrontEndPreview {type=NiagaraBool}
  User.ForwardDot {type=NiagaraFloat, defaultValue=0, description=""}
  User.RibbonIDOffset {type=NiagaraInt32}
  User.RightDot {type=NiagaraFloat, defaultValue=0, description=""}
  User.Variant {type=NiagaraFloat, defaultValue=0, description=""}
 ]
}
