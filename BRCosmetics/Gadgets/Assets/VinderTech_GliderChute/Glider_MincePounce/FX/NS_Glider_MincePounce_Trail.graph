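# Niagara trail effect for the MincePounce glider: socket-attached thruster meshes and flares, glow sprites with lights, torus-spawned speed lines and two ribbon trails.
# System timelines and a flicker alpha drive thruster scale, brightness and line spawn rate.
# Ribbons are burst particles repositioned each update along a slerped owner-axis vector, bent sideways by smoothed RightDot.
# Front-end modifiers swap preview values, and variant selectors switch color schemes.
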NiagaraSystem /BRCosmetics/Gadgets/Assets/VinderTech_GliderChute/Glider_MincePounce/FX/NS_Glider_MincePounce_Trail
dependencies {
 usedDataInterfaces [
  &NiagaraDataInterfaceCurve
  &NiagaraDataInterfacePlatformSet
  &NiagaraDataInterfaceVector2DCurve
  &NiagaraDataInterfaceColorCurve
  &NiagaraDataInterfaceSocketReader
  &NiagaraDataInterfaceArrayFloat
 ]
 usedDynamicInputs [
  &RemapRange
  &SelectFloatFromArray
  &MakeIntFromFloat
  &Waveform
 ]
 usedModules [
  &self:SystemSpawnScript.SystemScriptSource.SystemScriptGraph.NiagaraNodeAssignment_1.SetVariables_#1
  &SystemState
  &FN_Integrator
  &self:SystemSpawnScript.SystemScriptSource.SystemScriptGraph.NiagaraNodeAssignment_0.SetVariables_#2
  &FN_VariantManager
  &self:SystemSpawnScript.SystemScriptSource.SystemScriptGraph.NiagaraNodeAssignment_2.SetVariables_#3
 ]
}
emitters {
 Glows {
  emitterUpdateScript [
   &EmitterState(Life Cycle Mode=::NewEnumerator1, Loop Behavior=::NewEnumerator1, Loop Delay=1, Loop Duration Mode=::NewEnumerator1, UseLoopDelay=false)
   &SpawnBurst_Instantaneous(Spawn Count=4, Use Spawn Probability=false, UseLoopCountLimit=true)
   /SetVariables(
    Emitter.LightColor:LinearColor=&MultiplyLinearColor(
     Linear Color=&VariantColorSelector(Variant Amount=0, Original Color=<rgba 9.1332 13.372 60 1>, Variant 1 Color=<rgba 9.1332 43.0444 60 1>, Variant 2 Color=?)
     Scale Factor=&MakeVector4FromVectorAndFloat(
      "Vector (RGB)"=<1 1 1>
      "Float (Alpha)"=&FN_FrontEndFloatModifier(Gameplay=&Lerp_Float(A=0.2, B=1, Alpha=$System.Timeline.NormalizedTime), Front End=1)
     )
    )
    Emitter.PositionLight:NiagaraPosition=<0 90 2>
   ) {isSetParameters=true}
  ]
  particleSpawnScript [
   &PartitionParticles(Number Of Partitions=4)
   &InitializeParticle(
    Color=&VariantColorSelector(Variant Amount=0, Original Color=<rgba 1.5 2.9167 10 1>, Variant 1 Color=<rgba 1.5 7.1667 10 1>, Variant 2 Color=?)
    Sprite Size Mode=::NewEnumerator3
    Uniform Sprite Size=&MergeHelper_FloatSwitch(
     0=60
     1=35
     2=?
     3=?
     Index=/Niagara/DynamicInputs/Divide/Divide_Int(A:NiagaraInt32=$Particles.PartitionParticles.Partition, B:NiagaraInt32=2)
     Num=::NewEnumerator0
    )
    UsePositionOffset=true
    Write Material Random=true
   )
   &CameraOffset(Camera Offset Amount=-2)
   &SocketLocation(
    Interpolate Socket Sampling=true
    SamplePosOffset=true
    Sampled Position Offset=&MergeHelper_VectorSwitch(0=<-12 20 20>, 1=<12 20 20>, 2=<-37 25 8>, 3=<37 25 8>, Index=$Particles.PartitionParticles.Partition, Num=::NewEnumerator2)
    Socket Reader=$System.Socket Reader
    Socket Sampling Mode=::NewEnumerator3
    VelocitySpeedLimit=false
   )
  ]
  particleUpdateScript [
   &ParticleState(Kill Particles When Lifetime Has Elapsed=false, Let Infinitely Lived Particles Die When Emitter Deactivates=true)
   &ScaleColor(
    Scale Alpha=&Multiply_Float(
     A=$System.Timeline003.NormalizedTime
     B=&Multiply_Float(
      A=&FN_FrontEndFloatModifier(Gameplay=&Lerp_Float(A=0.2, B=1, Alpha=$System.Timeline.NormalizedTime), Front End=1)
      B=$System.ThrusterFlickerAlpha
     )
    )
    ScaleA=true
    ScaleRGB=true
   )
  ]
  rendererClasses [&NiagaraSpriteRendererProperties, &NiagaraLightRendererProperties]
  renderers [
   {rendererClass=&NiagaraSpriteRendererProperties}
   {rendererClass=&NiagaraLightRendererProperties, rendererIndex=1}
  ]
 }
 Speed_Lines {
  emitterUpdateScript [
   &EmitterState()
   &SpawnRate(
    SpawnRate=&FN_FrontEndFloatModifier(Gameplay=&Lerp_Float(A=0, B=90, Alpha=$System.Timeline.NormalizedTime), Front End=50)
    Use Spawn Probability=true
   ) {module=SpawnRate001}
  ]
  particleSpawnScript [
   &InitializeParticle(
    Color=&MakeLinearColorFromVectorAndFloat(
     "Vector (RGB)"=<1 1 1>
     "Float (Alpha)"=&Multiply_Float(
      A=&UniformRangedFloat(Minimum=0.1, Maximum=0.15)
      B=&FN_FrontEndFloatModifier(Gameplay=0.5, Front End=0.5)
     )
    )
    Lifetime Max=0.16
    Lifetime Min=0.1
    Lifetime Mode=::NewEnumerator1
    Sprite Size Max=<3 10>
    Sprite Size Min=<2 2>
    Sprite Size Mode=::NewEnumerator2
    UsePositionOffset=false
    Write Material Random=true
   )
   &TorusLocation(
    Align Torus With Axis=true
    Handle Radius=0
    Large Radius:NiagaraFloat=&FN_FrontEndFloatModifier(
     Gameplay=&RandomRangeFloat(Minimum=150, Maximum=200)
     Front End=&RandomRangeFloat(Minimum=180, Maximum=220)
    )
    MaskSpawnGroup=false
    Module.Override Randomness=false
    Torus Axis=&FN_FrontEndVector3Modifier(Gameplay=<0 1 0>, Front End=<1 0 0>)
    Torus Origin:NiagaraPosition=&AddVectorToPosition(
     Position=$Particles.Position
     Vector=&FN_FrontEndVector3Modifier(Gameplay=<0 0 0>, Front End=<300 0 0>)
    )
   )
   &AddVelocityFromPoint(
    Module.Use Falloff Distance=false
    Origin Offset=&FN_FrontEndVector3Modifier(Gameplay=<0 -500 0>, Front End=<0 0 0>)
    Velocity Strength=&Multiply_Float(
     A=&FN_FrontEndFloatModifier(Gameplay=600, Front End=-100)
     B=&MakeCustomFloatFromBool(Bool=$User.bIsFrontEndPreview, True Float=-1, False Float=1)
    )
   )
   &AddVelocity(Velocity=&Multiply_Vector(
     A=&FN_FrontEndVector3Modifier(
      Gameplay=&RandomRangeVector(Minimum=<0 300 0>, Maximum=<0 700 0>)
      Front End=&RandomRangeVector(Minimum=<300 0 0>, Maximum=<600 0 0>)
     )
     B=&MakeVector(X=&FN_FrontEndFloatModifier(Gameplay=1, Front End=-1), Y=1, Z=1)
    ))
  ]
  particleUpdateScript [
   &SolveForcesAndVelocity(Clamp Velocity=false, Limit Acceleration=false)
   &ParticleState()
   &ScaleSpriteSize(Scale Factor=&Vector2DFromCurve(
     Vector2Curve=DataInterface() {propertyValues {XCurve {keys [<at 0 0.5 RCIM_Cubic>,<at 0.5 1 RCIM_Cubic>]}, YCurve {keys [<at 0 0 RCIM_Cubic>,<at 0.5 1 RCIM_Cubic>]}}}
     ScaleCurve=&Vector2DFromFloat(Value=1)
    ))
   &ScaleSpriteSizeBySpeed(Max Scale Factor=<1 50>, Min Scale Factor=<1 40>, Velocity Threshold=100)
   &ScaleColor(
    Scale Alpha=&FN_FrontEndFloatModifier(
     Gameplay=&Lerp_Float(A=0, B=1, Alpha=$System.Timeline.NormalizedTime)
     Front End=&FloatFromCurve(
      FloatCurve=DataInterface() {propertyValues {Curve {
         defaultValue=1
         keys [{interpMode=RCIM_Cubic, tangentMode=RCTM_User, time=2.3, value=0}, {interpMode=RCIM_Cubic, tangentMode=RCTM_User, time=2.8, value=1}]
        }}}
      CurveIndex=$Emitter.Age
      Scale Curve=0.5
     )
    )
    Scale RGB=&VectorFromFloat(Value=1)
    ScaleA=true
    ScaleRGB=true
   )
  ]
  rendererClasses [&NiagaraSpriteRendererProperties]
  renderers [{rendererClass=&NiagaraSpriteRendererProperties}]
 }
 Thrsuters {
  emitterUpdateScript [
   &EmitterState(Life Cycle Mode=::NewEnumerator1, Loop Behavior=::NewEnumerator1, Loop Delay=1, Loop Duration Mode=::NewEnumerator1, UseLoopDelay=false)
   &SpawnBurst_Instantaneous(Spawn Count=3, Use Spawn Probability=false, UseLoopCountLimit=true)
  ]
  particleSpawnScript [
   &PartitionParticles(Number Of Partitions=3)
   &InitializeParticle(
    Color=&VariantColorSelector(Variant Amount=0, Original Color=<rgba 0.2625 0.15 1.5 1>, Variant 1 Color=<rgba 1.2875 0.225 1.5 1>, Variant 2 Color=?)
    Mesh Scale=&Multiply_VectorByFloat(Vector=<1 1 1>, Float=1)
    Mesh Scale Mode=::NewEnumerator4
    Position Offset=<0 10 19>
    Sprite Size=<300 80>
    Sprite Size Mode=::NewEnumerator4
    UsePositionOffset=false
    Write Material Random=true
   )
   &CameraOffset(Camera Offset Amount=4)
   /SetVariables(Particles.CameraOffsetFlare=50) {isSetParameters=true}
   /SetVariables(Particles.VisibilityTag=&MergeHelper_IntSwitch(0=0, 1=1, 2=1, 3=?, 4=?, 5=?, Index=$Particles.PartitionParticles.Partition, Num=::NewEnumerator1)) {isSetParameters=true}
   /SetVariables(Particles.CameraOffsetSideThrusters=1) {isSetParameters=true}
  ]
  particleUpdateScript [
   &SocketLocation(
    Interpolate Socket Sampling=true
    SamplePosOffset=true
    Sampled Position Offset=&MergeHelper_VectorSwitch(0=<0 16 18.5>, 1=<-0.3 0 0>, 2=<0.3 0 0>, 3=?, Index=$Particles.PartitionParticles.Partition, Num=::NewEnumerator1)
    Socket Index=&MergeHelper_IntSwitch(0=0, 1=3, 2=4, 3=?, 4=?, 5=?, Index=$Particles.PartitionParticles.Partition, Num=::NewEnumerator1)
    Socket Reader=$System.Socket Reader
    Socket Sampling Mode=::NewEnumerator3
    VelocitySpeedLimit=false
   )
   &ParticleState(Kill Particles When Lifetime Has Elapsed=false, Let Infinitely Lived Particles Die When Emitter Deactivates=true)
   &ScaleColor(
    Scale Alpha=&Multiply_Float(
     A=$System.Timeline003.NormalizedTime
     B=&Multiply_Float(A=&Multiply_Float(A=$System.Timeline.NormalizedTime, B=1), B=$System.ThrusterFlickerAlpha)
    )
    ScaleA=true
    ScaleRGB=true
   )
   &ScaleMeshSize(Scale Factor=&MakeVector(
     X=1
     Y=&Multiply_Float(
      A=&FN_FrontEndFloatModifier(Gameplay=&Lerp_Float(A=0.6, B=1.1, Alpha=$System.Timeline.NormalizedTime), Front End=1)
      B=&Lerp_Float(
       A=1
       B=2.3
       Alpha=&FloatFromCurve(
        FloatCurve=DataInterface() {propertyValues {Curve {keys [
            {interpMode=RCIM_Cubic, tangentMode=RCTM_User, time=0, value=0}
            {arriveTangent=-3.7461, interpMode=RCIM_Cubic, leaveTangent=-3.7461, tangentMode=RCTM_User, time=0.1, value=1}
            <at 0.66 0>
           ]}}}
        CurveIndex=$System.Timeline003.NormalizedTime
       )
      )
     )
     Z=1
    ))
   /SetVariables(Particles.ScaleMainThruster:Vector3f=&Multiply_VectorByFloat(Vector=$Particles.Scale, Float=1.04)) {isSetParameters=true}
   /SetVariables(Particles.PositionFlare:NiagaraPosition=&AddVectorToPosition(
     Position=$Particles.Position
     Vector=&RotateVectorByQuaternion(Vector To Rotate=<0 10 0>, Quaternion=$Output.SocketLocation.SampledRotation)
    )) {isSetParameters=true}
   /SetVariables(Particles.ColorCore:LinearColor=&MultiplyLinearColor(
     Linear Color=&VariantColorSelector(
      Variant Amount=0
      Original Color=&LerpLinearColors(StartColor=<rgba 0.1211 0.1903 0.5365 1>, EndColor=<rgba 1 1 1 1>, LerpFactor=$System.Timeline.NormalizedTime)
      Variant 1 Color=&LerpLinearColors(StartColor=<rgba 0.1042 0.075 0.25 1>, EndColor=<rgba 0.35 0.3 0.6 1>, LerpFactor=$System.Timeline.NormalizedTime)
      Variant 2 Color=?
     )
     Scale Factor=&MakeVector4FromVectorAndFloat(
      "Vector (RGB)"=<1 1 1>
      "Float (Alpha)"=&Multiply_Float(
       A=&Multiply_Float(A=$System.Timeline003.NormalizedTime, B=&Lerp_Float(A=0.3, B=1, Alpha=$System.Timeline.NormalizedTime))
       B=&Multiply_Float(
        A=&FNTimeOfDayScalar(Night Scale=0.7, Morning Scale=0.9, Evening Scale=0.9)
        B=&FN_PickFloatDependOnPlatform(TargetPlatformActive=1, TargetPlatformNoActive=0.7)
       )
      )
     )
    )) {isSetParameters=true}
   /SetVariables(Particles.ColorGlow:LinearColor=&MultiplyLinearColor(
     Linear Color=&VariantColorSelector(Variant Amount=0, Original Color=<rgba 1.1458 1.1458 2 1>, Variant 1 Color=<rgba 0.8 0.8833 1 1>, Variant 2 Color=?)
     Scale Factor=&MakeVector4FromVectorAndFloat(
      "Vector (RGB)"=&VectorFromFloat(Value=&Lerp_Float(
        A=14
        B=1
        Alpha=&FloatFromCurve(
         FloatCurve=DataInterface() {propertyValues {Curve {keys [
             {arriveTangent=12.4808, interpMode=RCIM_Cubic, leaveTangent=12.4808, tangentMode=RCTM_User, time=0.25, value=0}
             {arriveTangent=-0.1717, interpMode=RCIM_Cubic, leaveTangent=-0.1717, tangentMode=RCTM_User, time=0.4, value=1}
            ]}}}
         CurveIndex=$System.Timeline003.NormalizedTime
        )
       ))
      "Float (Alpha)"=&Multiply_Float(
       A=&FNTimeOfDayScalar(Night Scale=0.7, Morning Scale=0.9, Evening Scale=0.9)
       B=&Multiply_Float(
        A=&Multiply_Float(
         A=&FloatFromCurve(
          FloatCurve=DataInterface() {propertyValues {Curve {keys [{interpMode=RCIM_Cubic, tangentMode=RCTM_User, time=0, value=0}, {interpMode=RCIM_Cubic, tangentMode=RCTM_User, time=0.15, value=1}]}}}
          CurveIndex=$System.Timeline003.NormalizedTime
         )
         B=&Lerp_Float(A=0.3, B=1, Alpha=$System.Timeline.NormalizedTime)
        )
        B=$System.ThrusterFlickerAlpha
       )
      )
     )
    )) {isSetParameters=true}
   /SetVariables(
    Particles.PositionThrusterSides:NiagaraPosition=&AddVectorToPosition(
     Position=$Particles.Position
     Vector=&RotateVectorByQuaternion(Vector To Rotate=<0 2 0>, Quaternion=$Output.SocketLocation.SampledRotation)
    )
    Particles.ScaleThrusterSides:Vector3f=&Multiply_Vector(A=$Particles.Scale, B=<0.99 0.8 1>)
   ) {isSetParameters=true}
   /SetVariables(
    Particles.ColorCoreSmall:LinearColor=&MultiplyLinearColor(
     Linear Color=$Particles.ColorCore
     Scale Factor=&MakeVector4FromVectorAndFloat(
      "Vector (RGB)"=&MakeVectorFromLinearColorRGB(Color=&VariantColorSelector(Variant Amount=0, Original Color=<rgba 1.9429 27.3025 30 1>, Variant 1 Color=<rgba 3.3333 0 10 1>, Variant 2 Color=?))
      "Float (Alpha)"=&FN_PickFloatDependOnPlatform(TargetPlatformActive=4, TargetPlatformNoActive=2)
     )
    )
    Particles.ScaleCoreSmall:Vector3f=&Multiply_VectorByFloat(Vector=$Particles.Scale, Float=0.8)
   ) {isSetParameters=true}
   /SetVariables(
    Particles.ColorCoreRound:LinearColor=&MultiplyLinearColor(
     Linear Color=$Particles.ColorGlow
     Scale Factor=&Multiply_Vector4(
      A=&MakeVector4_FromLinearColor(LinearColor=<rgba 1.2953 18.2017 20 5>)
      B=&MakeVector4FromVectorAndFloat(
       "Vector (RGB)"=<1 1 1>
       "Float (Alpha)"=&FN_FrontEndFloatModifier(Gameplay=&Lerp_Float(A=0, B=1, Alpha=$System.Timeline.NormalizedTime), Front End=0.55)
      )
     )
    )
    Particles.ColorFakeLightRound:LinearColor=&MakeLinearColorFromVectorAndFloat(
     "Vector (RGB)"=&MakeVectorFromLinearColorRGB(Color=&VariantColorSelector(Variant Amount=0, Original Color=<rgba 1.5 2.9167 10 1>, Variant 1 Color=<rgba 10 0.7 1.475 1>, Variant 2 Color=?))
     "Float (Alpha)"=&Multiply_Float(
      A=&FN_FrontEndFloatModifier(Gameplay=&Multiply_Float(A=$System.Timeline.NormalizedTime, B=1), Front End=1)
      B=$System.ThrusterFlickerAlpha
     )
    )
    Particles.ColorGlowRound:LinearColor=$Particles.ColorGlow
    Particles.PositionFakeLightRound:NiagaraPosition=&AddVectorToPosition(
     Position=$Particles.Position
     Vector=&RotateVectorByQuaternion(
      Vector To Rotate=&MergeHelper_VectorSwitch(0=<0 0 0>, 1=<-6 4 0>, 2=<6 -4 0>, 3=?, Index=$Particles.PartitionParticles.Partition, Num=::NewEnumerator1)
      Quaternion=$Output.SocketLocation.SampledRotation
     )
    )
    Particles.ScaleThrusterRound:Vector3f=&Multiply_Vector(
     A=&VectorFromFloat(Value=&MergeHelper_FloatSwitch(0=1, 1=1, 2=-1, 3=?, Index=$Particles.PartitionParticles.Partition, Num=::NewEnumerator1))
     B=&MakeVector(
      X=0.5
      Y=&FN_FrontEndFloatModifier(Gameplay=&Lerp_Float(A=1, B=2, Alpha=$System.Timeline.NormalizedTime), Front End=2)
      Z=0.5
     )
    )
    Particles.SpriteSizeFakeLightRound:Vector2f=&Multiply_Vector2D_ByFloat(Float=50, Vector2D=<1 1>)
   ) {isSetParameters=true}
   /SetVariables(Particles.ScaleThrusterCoreRound:Vector3f=&Multiply_VectorByFloat(Vector=$Particles.ScaleThrusterRound, Float=0.5)) {isSetParameters=true}
  ]
  rendererClasses [&NiagaraSpriteRendererProperties, &NiagaraMeshRendererProperties]
  renderers [
   {rendererClass=&NiagaraSpriteRendererProperties}
   {rendererClass=&NiagaraMeshRendererProperties, rendererIndex=1}
   {rendererClass=&NiagaraMeshRendererProperties, rendererIndex=2}
   {rendererClass=&NiagaraMeshRendererProperties, rendererIndex=3}
   {rendererClass=&NiagaraMeshRendererProperties, rendererIndex=4}
   {rendererClass=&NiagaraMeshRendererProperties, rendererIndex=5}
   {rendererClass=&NiagaraSpriteRendererProperties, rendererIndex=6}
   {rendererClass=&NiagaraMeshRendererProperties, rendererIndex=7}
   {rendererClass=&NiagaraMeshRendererProperties, rendererIndex=8}
   {rendererClass=&NiagaraMeshRendererProperties, rendererIndex=9}
   {rendererClass=&NiagaraMeshRendererProperties, rendererIndex=10}
   {rendererClass=&NiagaraMeshRendererProperties, rendererIndex=11}
   {rendererClass=&NiagaraMeshRendererProperties, rendererIndex=12}
  ]
 }
 Trail {
  emitterUpdateScript [
   &EmitterState(
    Inactive Response=::NewEnumerator1
    Life Cycle Mode=::NewEnumerator1
    Loop Behavior=::NewEnumerator1
    Loop Duration Mode=::NewEnumerator1
    UseLoopDelay=false
   )
   &Timeline(Play=&Boolean_OrOperation(Bool1=&SetBoolByFloatComparison(A=$User.ForwardDot, B=0.1), Bool2=$User.bIsFrontEndPreview))
   &InterpolateOverTime(Target Float Value:NiagaraFloat=$User.RightDot)
   /SetVariables(
    Emitter.PositionOffset:Vector3f=&TransformVector(Vector=<0 0 0>, Destination Space=::Simulation)
    Emitter.RibbonColor:LinearColor=&VariantColorSelector(Variant Amount=0, Original Color=<rgba 0.388 0.4859 2 0.3>, Variant 1 Color=<rgba 0.6567 0.388 2 0.3>, Variant 2 Color=?)
    Emitter.RibbonLength:NiagaraFloat=&FN_FrontEndFloatModifier(Gameplay=&Lerp_Float(A=30, B=100, Alpha=$StackContext.Timeline.Time), Front End=120)
   ) {isSetParameters=true}
   &SpawnBurst_Instantaneous(Spawn Count=8, Use Spawn Probability=false, UseLoopCountLimit=false)
   /BRCosmetics/Animation/Game/MainPlayer/Emotes/UpbeatIguana/FX/InterpolateTrailVector(Range=0.3, Target Vector:Vector3f=$Engine.Owner.SystemYAxis)
  ]
  particleSpawnScript [
   &PartitionParticles(Particles Per Partition=4, Partition Mode:ENiagara_PartitionMode_v2=::NewEnumerator1, Write Partition to Ribbon ID=::NewEnumerator0)
   /SetVariables(Particles.RibbonWidth=&Multiply_Float(
     A=&Multiply_Float(A=12, B=&Lerp_Float(A=0.3, B=1, Alpha=$Particles.PartitionParticles.Normalized Index In Partition))
     B=&MergeHelper_FloatSwitch(0=1, 1=1, 2=?, 3=?, Index=$Particles.PartitionParticles.Partition, Num=::NewEnumerator0)
    )) {isSetParameters=true}
  ]
  particleUpdateScript [
   &SocketLocation(
    Interpolate Socket Sampling=true
    SamplePosOffset=true
    Sampled Position Offset=&MergeHelper_VectorSwitch(0=<0 -20 0>, 1=<0 -20 0>, 2=?, 3=?, Index=$Particles.PartitionParticles.Partition, Num=::NewEnumerator0)
    Socket Index=&MergeHelper_IntSwitch(0=1, 1=2, 2=?, 3=?, 4=?, 5=?, Index=$Particles.PartitionParticles.Partition, Num=::NewEnumerator0)
    Socket Reader=$System.Socket Reader
    Socket Sampling Mode=::NewEnumerator3
    VelocitySpeedLimit=false
   )
   /SetVariables(Particles.Color=&MultiplyLinearColor(
     Linear Color=$Emitter.RibbonColor
     Scale Factor=&MakeVector4FromVectorAndFloat(
      "Vector (RGB)"=<1 1 1>
      "Float (Alpha)"=&Multiply_Float(
       A=$System.Timeline003.NormalizedTime
       B=&FN_FrontEndFloatModifier(
        Gameplay=&Multiply_Float(
         A=$System.Timeline.Time
         B=&FloatFromCurve(
          FloatCurve=DataInterface() {propertyValues {Curve {keys [{interpMode=RCIM_Cubic, tangentMode=RCTM_User, time=0.1, value=0}, {interpMode=RCIM_Cubic, tangentMode=RCTM_User, time=0.5, value=1}]}}}
          CurveIndex=&OneMinusFloat(Float=$Particles.PartitionParticles.Normalized Index In Partition)
         )
        )
        Front End=1
       )
      )
     )
    )) {isSetParameters=true}
   /SetVariables(
    Particles.InterpolatedTrailVector:Vector3f=&TransformVector(
     Vector=/Niagara/DynamicInputs/Math/SlerpVector(
      Vector A:Vector3f=$Emitter.InterpolatedVector
      Vector B:Vector3f=$Emitter.TargetVector
      Alpha:NiagaraFloat=$Particles.PartitionParticles.Normalized Index In Partition
      Scale:NiagaraFloat=1
     )
     Source Space=::World
     Destination Space=::Simulation
     Apply Scale=true
    )
    Particles.RibbonFacing=&TransformVector(Vector=$Engine.Owner.SystemZAxis, Source Space=::World, Destination Space=::Simulation)
   ) {isSetParameters=true}
   /SetVariables(Particles.SidewaysOffsetVector:Vector3f=&Multiply_VectorByFloat(
     Vector=&NormalizeVector(
      Vector To Normalize=/Niagara/DynamicInputs/Vectors/CrossProduct(
       Vector A:Vector3f=$StackContext.InterpolatedTrailVector
       Vector B:Vector3f=$Particles.RibbonFacing
       Normalize Incoming Vectors:NiagaraBool=true
      )
      Safe Normalize=true
      Safe Fallback Threshold=0
      Safe Fallback Vector=<0 0 0.001>
     )
     Float=&Multiply_Float(
      A=$Emitter.InterpolateOverTime.InterpolateOverTime_Float.Moving Average
      B=&Multiply_Float(
       A=&Power(a=&OneMinusFloat(Float=$Particles.PartitionParticles.Normalized Index In Partition), b=3)
       B=-30
      )
     )
    )) {isSetParameters=true}
   /SetVariables(Particles.InterpolatedPosition:NiagaraPosition=&FN_FrontEndVector3Modifier(
     Gameplay=&AddVectorToPosition(
      Position=&AddVectorToPosition(
       Position=$Output.SocketLocation.SampledPosition
       Vector=&Add_Vector(A=$Emitter.PositionOffset, B=$StackContext.SidewaysOffsetVector)
      )
      Vector=&Multiply_VectorByFloat(
       Vector=$Particles.InterpolatedTrailVector
       Float=&Multiply_Float(A=&OneMinusFloat(Float=$Particles.PartitionParticles.Normalized Index In Partition), B=$Emitter.RibbonLength)
      )
     )
     Front End=&AddVectorToPosition(
      Position=&AddVectorToPosition(Position=$Output.SocketLocation.SampledPosition, Vector=<25 0 0>)
      Vector=&RotateVectorByQuaternion(
       Vector To Rotate=&MakeVector(
        X=0
        Y=&Multiply_Float(
         A=1
         B=&Multiply_Float(A=&OneMinusFloat(Float=$Particles.PartitionParticles.Normalized Index In Partition), B=$Emitter.RibbonLength)
        )
        Z=0
       )
       Quaternion=$Output.SocketLocation.SampledRotation
      )
     )
    )) {isSetParameters=true}
   /SetVariables(Particles.RibbonTwist:NiagaraFloat=&MergeHelper_FloatSwitch(0=0, 1=-1, 2=?, 3=?, Index=$Particles.PartitionParticles.Partition, Num=::NewEnumerator0)) {enabled=false, isSetParameters=true}
   /SetVariables(Particles.RibbonFacing=&FN_FrontEndVector3Modifier(
     Gameplay=&MergeHelper_VectorSwitch(0=<-0.3 0 1>, 1=<0.3 0 1>, 2=?, 3=?, Index=$Particles.PartitionParticles.Partition, Num=::NewEnumerator0)
     Front End=&MergeHelper_VectorSwitch(0=<0 -0.4 1>, 1=<0 0.4 1>, 2=?, 3=?, Index=$Particles.PartitionParticles.Partition, Num=::NewEnumerator0)
    )) {isSetParameters=true}
  ]
  rendererClasses [&NiagaraRibbonRendererProperties]
  renderers [{rendererClass=&NiagaraRibbonRendererProperties}]
 }
}
summary {
 systemName=NS_Glider_MincePounce_Trail
 emitters [
  {emitterName=Thrsuters, rendererClasses [&NiagaraSpriteRendererProperties, &NiagaraMeshRendererProperties]}
  {emitterName=Glows, rendererClasses [&NiagaraSpriteRendererProperties, &NiagaraLightRendererProperties]}
  {emitterName=Speed_Lines, rendererClasses [&NiagaraSpriteRendererProperties]}
  {emitterName=Trail, rendererClasses [&NiagaraRibbonRendererProperties]}
 ]
 userVariables [
  User.bIsFrontEnd {type=NiagaraBool}
  User.bIsFrontEndPreview {type=NiagaraBool}
  User.Boost {type=NiagaraFloat, defaultValue=0, description=""}
  User.EnableSpawn {type=NiagaraBool}
  User.ForwardDot {type=NiagaraFloat, defaultValue=0, description=""}
  User.ForwardSpeed {type=NiagaraFloat, defaultValue=0, description=""}
  User.RibbonIDOffset {type=NiagaraInt32}
  User.RightDot {type=NiagaraFloat, defaultValue=0, description=""}
  User.TurnDirection {type=NiagaraFloat, defaultValue=0, description=""}
  User.TurningAngle {type=NiagaraFloat, defaultValue=0, description=""}
  User.TurnSpeed {type=NiagaraFloat, defaultValue=0, description=""}
  User.Variant {type=NiagaraFloat, defaultValue=0, description=""}
 ]
}
